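; Thrown TNT projectile built on the grenade base, with a timed fuse, looping fuse sound indicator and explosion effects.
; A replicated detonation time, offset by server time, times the fuse, which ends in a team-tinted radius effect and explosion.
; On stopping, the bundle rotates to keep the fuse visible, and simulation resumes if its supporting building or vehicle is destroyed.
; Cosmetic-only logic selects water or surface-based explosion sounds and effects.
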(event EventBeginPlay
  ;[1 1 1] Setup for Change to Particle System if needed
  (Variables|VFX|SetP_Explosion (select (Variables|Default|GetUse%20Shockwave%20Based%20Explosion) 0 0))
  (CallFunction|AysncLoad)
  (Networking|CanExecuteCosmeticEvents
    (:True
      (Audio|Components|Audio|SetSound (Variables|B_Prj_Grenade|GetGrenadeFuse_AudioComponent) (Variables|Sounds|GetCue_GrenadeFuseSound))
      (if (Variables|Default|GetbShowFuseSoundIndicator)
        (bind _returnvalue (Utilities|Time|SetTimerbyEvent (delegate Custom|LoopingIndicator) 0.3 true))
        (Variables|Default|SetFuseIndicatorTimer _returnvalue)))))

(event Custom|LoopingIndicator
  ;[1 1 1] Setup for Change to Particle System if needed
  (Fortnite|ShowSoundIndicatorLocal :TrackedActor self :Location (Transformation|GetActorLocation) :MaxAudibleDistance (Variables|/Script/Engine.SoundBase|GetMaxDistance (Variables|Sounds|GetCue_GrenadeFuseSound)) :IndicatorType "Glider" :Tint "(R=1.000000,G=1.000000,B=1.000000,A=1.000000)" :OverrideIcon (Variables|Default|GetSoundIndicatorIcon) :AttenuationMaxRangeMultiplier 1.0 :AttenuationIntensityMultiplier 1.0) ; Indicator for fuse
)

(event EventOnStop
  (params (Hit :type Hit%20Result%20Structure%20%28by%20ref%29))
  (bind (_bblockinghit _binitialoverlap _time _distance _location _impactpoint _normal) (Collision|BreakHitResult Hit))
  ;[1 1 1] Keeps Fuse Visible when projectile hits ground
  (|CallParentFunction Hit)
  (Networking|CanExecuteCosmeticEvents
    (:True
      (bind _outrotation (CallFunction|MakeOrthogonalZ (Transformation|GetWorldRotation (Variables|Default|GetTNTBundle)) _normal))
      (Variables|Default|SetTargetRotation _outrotation)
      (bind _returnvalue (Utilities|Time|SetTimerbyEvent (delegate Custom|Lerp) 0.0333 true))
      (Variables|Default|SetLerpTimer _returnvalue)
      (Utilities|FlowControl|Delay (Variables|Default|GetLerpDuration))
      (Utilities|Time|ClearandInvalidateTimerbyHandle _returnvalue))))

(event Custom|Lerp
  (bind _tntbundle (Variables|Default|GetTNTBundle))
  (bind _returnvalue (Transformation|GetWorldRotation _tntbundle))
  (bind _targetrotation (Variables|Default|GetTargetRotation))
  ;[1 1 1] Keeps Fuse Visible when projectile hits ground
  (if (Math|Rotator|Equal(Rotator) _returnvalue _targetrotation 0.01)
    (Utilities|Time|ClearandInvalidateTimerbyHandle (Variables|Default|GetLerpTimer))
    (else
      (Transformation|SetWorldRotation (Variables|Default|GetTNTBundle) (Math|Interpolation|RInterpTo _returnvalue _targetrotation (Utilities|Time|GetWorldDeltaSeconds) (Variables|Default|GetInterpSpeed)) false false))))

(event OnProjectileStop(ProjectileMovementComponent)
  (params (ImpactResult :type Hit%20Result%20Structure%20%28by%20ref%29))
  (bind (_bblockinghit _binitialoverlap _time _distance _location _impactpoint _normal _impactnormal _physmat _hitactor) (Collision|BreakHitResult ImpactResult))
  ;[1 1 1] Bind events for resimulating after the attached building actor or vehicle is destroyed
  (bind _asbuilding_actor (Utilities|Casting|CastToBuildingActor _hitactor)
    (:then
      (EventDispatchers|BindEventtoOnDied _asbuilding_actor (delegate Custom|Event_BuildingActorDied)))
    (:CastFailed
      (bind _asfort_vehicle_interface (Utilities|Casting|CastToFortVehicleInterface _hitactor))
      (Game|BindEventtoOnDestroyed (Vehicle|GetActor _asfort_vehicle_interface) (delegate Custom|Event_OnVehicleDestroyed)))))

(event Custom|Event_BuildingActorDied
  (params (DamagedActor :type Actor%20Object%20Reference) (Damage :type Float%20%28single-precision%29) (InstigatedBy :type Controller%20Object%20Reference) (DamageCauser :type Actor%20Object%20Reference) (HitLocation :type Vector) (FHitComponent :type Primitive%20Component%20Object%20Reference) (BoneName :type Name) (Momentum :type Vector))
  ;[1 1 1] Bind events for resimulating after the attached building actor or vehicle is destroyed
  (Projectile|ResumeSimulation))

(event EventDestroyed
  ;[1 1 1] End lerp timer on destroyed
  (Utilities|Time|ClearandInvalidateTimerbyHandle (Variables|Default|GetLerpTimer))
  (|CallParentFunction))

(event Custom|Event_StartFuse
  (params (DetonationTime :type Float%20%28double-precision%29))
  ;[1 1 1] Set event to explode when the passed in remaining fuse time expires
  (Variables|Default|SetDetonationTime DetonationTime))

(event Custom|Event_OnVehicleDestroyed
  (params (DestroyedActor :type Actor%20Object%20Reference))
  ;[1 1 1] Bind events for resimulating after the attached building actor or vehicle is destroyed
  (Projectile|ResumeSimulation))

(event EventFuseEnded
  (bind _effect_distance (Variables|B_Prj_Grenade|GetEffect_Distance))
  (Transformation|SetWorldRotation _effect_distance "90.000000,0.000000, 0" false false)
  (Components|Activation|Activate _effect_distance false)
  (Effects|Components|ParticleSystem|SetVectorParameter _effect_distance "Size" (Math|Conversions|ToVector(Float) (Variables|Default|GetExplosionRadius)))
  (if (not (Team|OnSameTeam (Game|GetInstigatorController) (Game|GetPlayerController 0)))
    (Effects|Components|ParticleSystem|SetVectorParameter _effect_distance "Color" "1.500000,0.110000,0.190000"))
  (Delivery|DoExplosion))

(event OnEnterWaterEvent(WaterInteractionComponent)
  (params (WaterBody :type Fort%20Water%20Body%20Actor%20Object%20Reference) (WaterInteractionComponent :type Fort%20Water%20Interaction%20Component%20Object%20Reference) (bIsFirstBody :type Boolean))
  ;[0.8698 0.08154 0.2343 0.5] Hit Water Surface reaction.  Added this because the bounce doesn’t occur in the parent BP.
  (if bIsFirstBody
    (CallFunction|HeldWaterImpact)
    (Utilities|FlowControl|DoOnce false
      (:Completed
        (CallFunction|WaterSurfaceHitFX)))))

(event Custom|WaterSurfaceHitFX
  ;[0.8698 0.08154 0.2343 0.5] Hit Water Surface reaction.  Added this because the bounce doesn’t occur in the parent BP.
  (Effects|Components|ParticleSystem|SpawnFXAtLocation (Variables|VFX|GetNS_HitWater) (Transformation|GetActorLocation) (Math|Rotator|MakeRotator 0.0 90))
  (CallFunction|StopRotation))

(event EventOnExploded
  (params (HitActors :type Array%20of%20Actor%20Object%20References) (HitResults :type Array%20of%20Hit%20Result%20Structures))
  (bind _returnvalue (Transformation|GetActorLocation))
  (bind _actual_radius (Variables|Default|GetActual%20Radius))
  (bind _returnvalue_2 (/ _actual_radius 5e+02))
  (bind _returnvalue_3 (Transformation|GetActorLocation))
  (bind _returnvalue_4 (Transformation|GetActorLocation))
  (bind _returnvalue_5 (Math|Rotator|MakeRotator 0.0 90))
  (bind _cue_explosionsound (Variables|Sounds|GetCue_ExplosionSound))
  ;[1 1 1] OnExploded - Emitter, Sound, Camera Shake
  (Utilities|Time|ClearandInvalidateTimerbyHandle (Variables|Default|GetFuseIndicatorTimer))
  (Utilities|Time|ClearandInvalidateTimerbyHandle (Variables|Default|GetLerpTimer))
  (Components|Movement|SetUpdatedComponent (Variables|Movement|GetProjectileMovementComponent) 0)
  (Projectile|RemoveGameplayCue%28Looping%29Local self (Variables|Default|GetFuseCue) (Ability|GameplayCue|MakeGameplayCueParameters :NormalizedMagnitude 0.0 :RawMagnitude 0.0 :EffectContext "()" :AggregatedSourceTags "(GameplayTags=)" :AggregatedTargetTags "(GameplayTags=)" :Location "0,0,0" :Normal "0,0,0" :GameplayEffectLevel 1 :AbilityLevel 1 :bReplicateLocationWhenUsingMinimalRepProxy false))
  (Networking|CanExecuteCosmeticEvents
    (:True
      (bind _outhit (Collision|LineTraceByChannel (+ _returnvalue_4 "0.000000,0.000000,50.000000") (+ _returnvalue_4 "0.000000,0.000000,-50.000000") "TraceTypeQuery12"))
      (bind (_outhit_bblockinghit _outhit_binitialoverlap _outhit_time _outhit_distance _outhit_location _outhit_impactpoint _outhit_normal _outhit_impactnormal _outhit_physmat _outhit_hitactor _outhit_hitcomponent) (Collision|BreakHitResult _outhit))
      (if (Utilities|String|Contains (Utilities|GetDisplayName _outhit_hitcomponent) "waterbody")
        (bind _returnvalue_6 (Effects|Components|ParticleSystem|SpawnFXAtLocation (CallFunction|GetExplosion true) _returnvalue_4 _returnvalue_5))
        (Rendering|SetRenderCustomDepth _returnvalue_6 true)
        (Audio|PlaySoundatLocation (Variables|Sounds|GetCue_ExplosionSound) _returnvalue)
        (Audio|PlaySoundatLocation (Variables|Default|GetCue%20Water%20Sound) _returnvalue)
        (else
          (bind _returnvalue_7 (Effects|Components|ParticleSystem|SpawnFXAtLocation (CallFunction|GetExplosion false) _returnvalue_4 _returnvalue_5))
          (Rendering|SetRenderCustomDepth _returnvalue_7 true)
          (Components|Activation|Deactivate (Variables|Default|GetNS_Fuse_Particle))
          (if (Variables|Default|GetUse%20Shockwave%20Based%20Explosion)
            (Effects|Components|ParticleSystem|SetVectorParameter _returnvalue_7 "RadiusSize" (Math|Conversions|ToVector(Float) _returnvalue_2))
            (Effects|Components|ParticleSystem|SetFloatParameter _returnvalue_7 "SphereRadius" (* _returnvalue_2 1.5e+02))
            (Effects|Components|ParticleSystem|SetFloatParameter _returnvalue_7 "Global Scale" 0.95))
          (GameplayCue|ExecuteGameplayCue%28Burst%29Local self (Variables|Default|GetFeedbackCue) (Ability|GameplayCue|MakeGameplayCueParameters 0.0)) ; Cam shake + controller vibration
          (Class|BPSurfaceTypeSoundComponent|PlaySound (Variables|Default|GetBP_SurfaceTypeSoundComponent) (Transformation|GetActorLocation))
          (Audio|PlaySoundatLocation _cue_explosionsound _returnvalue_3) ; Play Explosion
        ))
      (Audio|Components|Audio|FadeOut (Variables|B_Prj_Grenade|GetGrenadeFuse_AudioComponent) 0.2 0.0 "Linear" :!class /Script/Engine.AudioComponent) ; Fade Out Fuse Sound
      (Fortnite|ShowSoundIndicatorLocal self _returnvalue_3 (Variables|/Script/Engine.SoundBase|GetMaxDistance _cue_explosionsound) "Gunshot")
      (Development|SetHiddeninGame (Variables|Default|GetTNTBundle) true false))))

(event EventOnBounce
  (params (Hit :type Hit%20Result%20Structure%20%28by%20ref%29))
  (bind (_bblockinghit _binitialoverlap _time _distance _location _impactpoint _normal _impactnormal _physmat _hitactor) (Collision|BreakHitResult Hit))
  (bind _surfacetype (Variables|/Script/PhysicsCore.PhysicalMaterial|GetSurfaceType _physmat))
  (|CallParentFunction Hit)
  (Variables|Default|SetStopLocation _impactpoint)
  (Variables|Default|SetSurface%20Type _surfacetype)
  (Class|BPSurfaceTypeSoundComponent|SetSurfaceType (Variables|Default|GetBP_SurfaceTypeSoundComponent) _surfacetype _hitactor))

(fn MakeOrthogonalZ
  (params (InRotation :type Rotator) (Normal :type Vector))
  (bind _returnvalue (Math|Vector|DotProduct (.Z (Math|Rotator|GetAxes InRotation)) Normal))
  (bind _returnvalue_2 (* Normal _returnvalue))
  (bind _returnvalue_3 (- (.Z (Math|Rotator|GetAxes InRotation)) _returnvalue_2))
  (bind _returnvalue_4 (Math|Vector|Normalize _returnvalue_3))
  (bind _returnvalue_5 (Math|Vector|DotProduct (.Y (Math|Rotator|GetAxes InRotation)) _returnvalue_4))
  (bind (_x _y _z) (Math|Rotator|GetAxes InRotation))
  (bind _returnvalue_6 (Math|Vector|DotProduct _x _returnvalue_4))
  (if (> (Math|Float|Absolute(Float) _returnvalue_6) (Math|Float|Absolute(Float) _returnvalue_5))
    (return (Math|Rotator|MakeRotfromZX _returnvalue_4 (Math|Vector|Normalize (- _x (* _returnvalue_4 _returnvalue_6)))))
    (else
      (return (Math|Rotator|MakeRotfromZY _returnvalue_4 (Math|Vector|Normalize (- _y (* _returnvalue_4 _returnvalue_5))))))))

(fn OnRep_DetonationTime
  (bind _output_get (Variables|Default|SetFuseTime (- (Variables|Default|GetDetonationTime) (Fortnite|Sync|Time|GetServerWorldTimeSeconds))))
  (Projectile|AddGameplayCue%28Looping%29Local self (Variables|Default|GetFuseCue) (Ability|GameplayCue|MakeGameplayCueParameters 0.0 _output_get))
  (bind _returnvalue (Utilities|Time|SetTimerbyEvent (EventDispatchers|CreateEvent :!event FuseEnded) _output_get))
  (Variables|Default|SetFuseTimer _returnvalue))

(fn SplinePointLocations
  (return (Utilities|Array|MakeArray (Math|Transform|InverseTransformLocation (Transformation|GetWorldTransform (Variables|Default|GetSpline)) (Transformation|GetSocketLocation (Variables|Default|GetTNTBundle) "FX_FuseBegin")) (Math|Transform|InverseTransformLocation (Transformation|GetWorldTransform (Variables|Default|GetSpline)) (Transformation|GetSocketLocation (Variables|Default|GetTNTBundle) "dyn_fuse2")) (Math|Transform|InverseTransformLocation (Transformation|GetWorldTransform (Variables|Default|GetSpline)) (Transformation|GetSocketLocation (Variables|Default|GetTNTBundle) "dyn_fuse1")) (Math|Transform|InverseTransformLocation (Transformation|GetWorldTransform (Variables|Default|GetSpline)) (Transformation|GetSocketLocation (Variables|Default|GetTNTBundle) "FX_FuseEnd")))))

(fn ConstructionScript
  (|CallParentFunction))
