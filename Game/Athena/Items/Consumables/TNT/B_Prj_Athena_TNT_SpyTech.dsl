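; SpyTech variant of the Athena TNT consumable's projectile, adding knockback to player pawns when it explodes.
; For each hit result it applies knockback with a scalable magnitude, then runs the parent explosion logic.
; The KnockBackEnabled multiplier chooses the direction: away from the hit location with Z forced to 1 when positive, otherwise straight up, so knockback applies either way.
; BeginPlay, fuse end and construction only defer to the parent.
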(event EventBeginPlay
  (|CallParentFunction))

(event EventFuseEnded
  (|CallParentFunction))

(event EventOnExploded
  (params (HitActors :type Array%20of%20Actor%20Object%20References) (HitResults :type Array%20of%20Hit%20Result%20Structures))
  (for _array_element HitResults
    (bind _asfort_player_pawn (Utilities|Casting|CastToFortPlayerPawn (.HitActor (Collision|BreakHitResult _array_element))))
    (if (> (ScalableFloat|GetScalableFloatMultiplier (Variables|Default|GetKnockBackEnabled)) 0.0)
      (bind (_returnvalue_x _returnvalue_y _returnvalue_z) (Math|Vector|BreakVector (Transformation|GetActorLocation _asfort_player_pawn)))
      (bind (_returnvalue_x_2 _returnvalue_y_2) (Math|Vector|BreakVector (Math|Vector|Normalize (Math|Vector|vector-vector (Math|Vector|MakeVector _returnvalue_x _returnvalue_y (+ _returnvalue_z (* (Components|Capsule|GetScaledCapsuleHalfHeight (Variables|/Script/Engine.Character|GetCapsuleComponent _asfort_player_pawn)) 2.0))) (.Location (Collision|BreakHitResult _array_element))))))
      (HitReaction|ApplyKnockback _asfort_player_pawn (ScalableFloat|GetValueatLevel (Variables|Default|GetKnockZMagnitude)) 45.0 (Math|Vector|MakeVector _returnvalue_x_2 _returnvalue_y_2 1.0))
      (else
        (HitReaction|ApplyKnockback _asfort_player_pawn (ScalableFloat|GetValueatLevel (Variables|Default|GetKnockZMagnitude)) 45.0 (0.0 0.0 1.0)))))
  (|CallParentFunction HitActors HitResults))

(fn ConstructionScript
  (|CallParentFunction))
